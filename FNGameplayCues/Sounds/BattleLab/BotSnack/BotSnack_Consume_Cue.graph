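# BotSnack_Consume_Cue is a SoundCue for consuming a Bot Snack in Battle Lab, routed to the Fort_Gadgets_Items sound class.
# A Random node picks one of three equally weighted consume waves and feeds it into a Modulator at the root.
# The cue plays at reduced volume with its own short-range attenuation and stereo spread, plus a concurrency override that caps simultaneous instances.
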SoundCue /FNGameplayCues/Sounds/BattleLab/BotSnack/BotSnack_Consume_Cue
root=Modulator_1
nodes [
 {id=Modulator_1, children [Random_0]}
 {id=Random_0, children [WavePlayer_1,WavePlayer_2,WavePlayer_3], properties {weights [1,1,1]}}
 {id=WavePlayer_1, properties {soundWaveAssetPtr=&/BotSnack_Consume_01}}
 {id=WavePlayer_2, properties {soundWaveAssetPtr=&/BotSnack_Consume_02}}
 {id=WavePlayer_3, properties {soundWaveAssetPtr=&/BotSnack_Consume_03}}
]
properties {
 bOverrideAttenuation=true
 bOverrideConcurrency=true
 maxDistance=1400
 soundClassObject=&Fort_Gadgets_Items
 volumeMultiplier=0.4
 attenuationOverrides {falloffDistance=1200, stereoSpread=500, attenuationShapeExtents {x=200}}
 concurrencyOverrides {maxCount=4}
}
