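# Sound cue for throwing a Battle Lab Bot Snack, classed under gadget items.
# A random node picks one of three equally weighted throw waves and feeds it through a modulator at the root.
# The cue lowers its volume, overrides attenuation for limited-range falloff, and overrides concurrency so new plays are blocked rather than cutting off playing ones.
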SoundCue /FNGameplayCues/Sounds/BattleLab/BotSnack/BotSnack_Throw_Cue
root=Modulator_0
nodes [
 {id=Modulator_0, children [Random_1]}
 {id=Random_1, children [WavePlayer_0,WavePlayer_4,WavePlayer_5], properties {weights [1,1,1]}}
 {id=WavePlayer_0, properties {soundWaveAssetPtr=&/BotSnack_Throw_01}}
 {id=WavePlayer_4, properties {soundWaveAssetPtr=&/BotSnack_Throw_02}}
 {id=WavePlayer_5, properties {soundWaveAssetPtr=&/BotSnack_Throw_03}}
]
properties {
 bOverrideAttenuation=true
 bOverrideConcurrency=true
 maxDistance=2000
 soundClassObject=&Fort_Gadgets_Items
 volumeMultiplier=0.7
 attenuationOverrides {falloffDistance=1800, stereoSpread=1000, attenuationShapeExtents {x=200}}
 concurrencyOverrides {resolutionRule=PreventNew}
}
